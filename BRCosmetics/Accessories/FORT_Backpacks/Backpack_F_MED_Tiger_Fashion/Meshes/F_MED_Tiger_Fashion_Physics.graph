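# Physics asset for the Tiger Fashion female medium backpack cosmetic, defining simulation bodies for its dangling accessories.
# Five box bodies cover the pack, two fur pieces, a sword and a knife.
# Each accessory is constrained directly to the pack in a star layout with no chaining.
# The fur swings widely with little twist, while the sword and knife are limited to a few degrees of wobble.
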PhysicsAsset /BRCosmetics/Accessories/FORT_Backpacks/Backpack_F_MED_Tiger_Fashion/Meshes/F_MED_Tiger_Fashion_Physics
bodies {
 dyn_fur_a [{shapeType=Box, extentX=4.4068, extentY=4.4651, extentZ=5.2632, center=<1.2991 0 0>, rotation=<rot 90 0 0>}]
 dyn_fur_b [{shapeType=Box, extentX=3.8699, extentY=3.7506, extentZ=5.4669, center=<0.7158 0 0>, rotation=<rot 90 0 0>}]
 dyn_knife [{shapeType=Box, extentX=3.7007, extentY=1.2982, extentZ=31.5893, center=<2.5039 0 0>, rotation=<rot 90 0 0>}]
 dyn_sword [{shapeType=Box, extentX=4.2564, extentY=3.3927, extentZ=56.6503, center=<9.0712 0 0>, rotation=<rot 90 0 0>}]
 pack [{shapeType=Box, extentX=2.4443, extentY=8.6529, extentZ=33.7009, center=<0.3795 0.9199 -0.4089>, rotation=<rot -1.9718 -90 20.9575>}]
}
constraints [
 {bone1Name=dyn_sword, bone2Name=pack, swing1LimitDegrees=3, swing2LimitDegrees=3, twistLimitDegrees=4}
 {bone1Name=dyn_fur_b, bone2Name=pack, swing1LimitDegrees=45, swing2LimitDegrees=45, twistLimitDegrees=5}
 {bone1Name=dyn_fur_a, bone2Name=pack, swing1LimitDegrees=45, swing2LimitDegrees=45, twistLimitDegrees=5}
 {bone1Name=dyn_knife, bone2Name=pack, swing1LimitDegrees=2.5, swing2LimitDegrees=2.5, twistLimitDegrees=4}
]
