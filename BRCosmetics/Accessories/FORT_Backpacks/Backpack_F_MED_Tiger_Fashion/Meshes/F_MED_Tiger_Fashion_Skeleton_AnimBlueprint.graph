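# Animation Blueprint for the F_MED_Tiger_Fashion backpack cosmetic skeleton.
# The active AnimGraph path offsets the input pose's root bone with a replaced translation and additive yaw in component space, then outputs it.
# A RigidBody node with a physics asset override and custom gravity is present but unconnected, so it does not affect the pose.
# The EventGraph holds only unwired default update nodes.
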AnimBlueprint /BRCosmetics/Accessories/FORT_Backpacks/Backpack_F_MED_Tiger_Fashion/Meshes/F_MED_Tiger_Fashion_Skeleton_AnimBlueprint
target_skeleton=&/F_MED_Tiger_Fashion_Skeleton
graphs [
 AnimGraph {nodes [
   RigidBody() {
    bClampLinearTranslationLimitToRefPose=true
    bOverrideWorldGravity=true
    overridePhysicsAsset=&/F_MED_Tiger_Fashion_Physics
    alphaBoolBlend {blendOption=Linear}
    componentAppliedLinearAccClamp=<3500 3500 3500>
    componentLinearAccScale=<0.6 0.6 0.6>
    componentLinearVelScale=<0.6 0.6 0.6>
    overrideWorldGravity {z=-980}
   }
   n1 = "Transform (Modify) Bone - Bone: root"(ComponentPose=Local To Component(LocalPose=Input Pose()), Translation="0,-2,0", Rotation="0,-8,0") {
    rotationMode=BMM_Additive
    translationMode=BMM_Replace
    alphaBoolBlend {blendOption=Linear}
    boneToModify {boneName=root}
    rotation {yaw=-8}
    translation {y=-2}
   }
   Output Pose(Result=Component To Local(ComponentPose=@n1)) {layerGroup=None}
  ]}
 EventGraph {nodes [
   TryGetPawnOwner()
   Event BlueprintUpdateAnimation()
  ]}
]
variables []
